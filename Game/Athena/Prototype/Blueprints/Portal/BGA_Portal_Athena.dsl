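; BGA_Portal_Athena is a prototype Fortnite portal that teleports player pawns to its linked ActorRef portal.
; An outer capsule drives opening visuals and audio, while OverlapCapsule triggers server-side teleports with immunity, effects, randomized exit offsets and non-Creative quest credit.
; A delayed timer has the linked portal re-teleport anyone still overlapping the exit.
; A scalable float can destroy both portals, and vehicle teleporting is stubbed out.
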(event Custom|PlaylistLoaded
  (params (PlaylistName :type Name) (PlaylistContextTags :type Gameplay%20Tag%20Container%20Structure%20%28by%20ref%29))
  ;[1 0.9658 0.04048] (NO QUEST STUFF YET) Check if In Creative Mode or not for Quest
  (Variables|QuestStuff|SetCreativeMode (GameplayTags|HasTag PlaylistContextTags "(TagName=\"Athena.Creative\")")))

(event EventBeginPlay
  (bind _actorref (Variables|Portal|GetActorRef))
  ;[1 1 1] EVENT--> BeginPlay
  (|CallParentFunction)
  ;;[1 1 1] Wait until everything is loaded.
  (bind _asfort_game_state_athena (Utilities|Casting|CastToFortGameStateAthena (Game|GetGameState)))
  (if (Variables|/Script/FortniteGame.FortGameStateAthena|GetbPlaylistDataIsLoaded _asfort_game_state_athena)
    (bind _asbga_portal_athena (Utilities|Casting|CastToBGA_Portal_Athena _actorref))
    (Variables|Default|SetActorArrow (Variables|Default|GetArrow _asbga_portal_athena))
    (Variables|Default|SetTargetCapsule (Variables|Default|GetOverlapCapsule _asbga_portal_athena))
    (Variables|Portal|SetTeleportLocation (+ (Transformation|GetWorldLocation (Variables|Default|GetArrow _asbga_portal_athena)) (* (Transformation|GetForwardVector (Variables|Default|GetArrow _asbga_portal_athena)) 1.8e+02)))
    (Variables|Portal|SetActorRotation (Transformation|GetWorldRotation (Variables|/Script/Engine.Actor|GetRootComponent _actorref)))
    (CallFunction|CheckShouldSpawn)
    (else
      (Playlist|BindEventtoOnCurrentPlaylistLoaded _asfort_game_state_athena (delegate Custom|OnLoaded)))))

(event Custom|OnLoaded
  (params (PlaylistName :type Name) (PlaylistContextTags :type Gameplay%20Tag%20Container%20Structure%20%28by%20ref%29))
  (bind _actorref (Variables|Portal|GetActorRef))
  ;[1 1 1] EVENT--> BeginPlay
  ;;[1 1 1] Wait until everything is loaded.
  (bind _asbga_portal_athena (Utilities|Casting|CastToBGA_Portal_Athena _actorref))
  (Variables|Default|SetActorArrow (Variables|Default|GetArrow _asbga_portal_athena))
  (Variables|Default|SetTargetCapsule (Variables|Default|GetOverlapCapsule _asbga_portal_athena))
  (Variables|Portal|SetTeleportLocation (+ (Transformation|GetWorldLocation (Variables|Default|GetArrow _asbga_portal_athena)) (* (Transformation|GetForwardVector (Variables|Default|GetArrow _asbga_portal_athena)) 1.8e+02)))
  (Variables|Portal|SetActorRotation (Transformation|GetWorldRotation (Variables|/Script/Engine.Actor|GetRootComponent _actorref)))
  (CallFunction|CheckShouldSpawn))

(event Custom|PlayTeleportFX
  (params (PlayerLocation :type Vector) (Player :type Fort%20Player%20Pawn%20Object%20Reference))
  ;[1 1 1] Spawn teleport visuals and audio for clients
  (if (not (Variables|Default|GetUseNiagaraEnterPortalFX))
    (Effects|SpawnEmitteratLocation (Variables|VFX|GetEnterPortal) PlayerLocation) ; Cover player/vehicle vanish with FX
    (Effects|SpawnEmitteratLocation (Variables|VFX|GetEnterPortal) (Variables|Portal|GetTeleportLocation))
    (else
      (bind _returnvalue (Niagara|SpawnSystematLocation (Variables|VFX|GetNS_EnterPortal) PlayerLocation))
      (Variables|Default|SetEnterPortalFX _returnvalue)))
  (GameplayCue|ExecuteGameplayCue%28Burst%29Local Player (Variables|Default|GetTeleportGameplayCue) (Ability|GameplayCue|MakeGameplayCueParameters :NormalizedMagnitude 0.0 :RawMagnitude 0.0 :EffectContext "()" :AggregatedSourceTags "(GameplayTags=)" :AggregatedTargetTags "(GameplayTags=)" :Location "0,0,0" :Normal "0,0,0" :GameplayEffectLevel 1 :AbilityLevel 1 :bReplicateLocationWhenUsingMinimalRepProxy false))
  (Utilities|IsValid Player
    (:Is%20Valid
      (if (Pawn|IsLocallyControlled Player)
        (Audio|PlaySound2D (Variables|Sound|GetEnterAndExitPortal1P))
        (Development|PrintString "Played 1P")
        (else
          (Audio|PlaySoundatLocation (Variables|Sound|GetEnteredPortal) PlayerLocation)
          (Audio|PlaySoundatLocation (Variables|Sound|GetExitedPortal) (Variables|Portal|GetTeleportLocation))
          (Development|PrintString "Played 3P"))))))

(event OnComponentBeginOverlap(Capsule)
  (params (OverlappedComponent :type Primitive%20Component%20Object%20Reference) (OtherActor :type Actor%20Object%20Reference) (OtherComp :type Primitive%20Component%20Object%20Reference) (OtherBodyIndex :type Integer) (bFromSweep :type Boolean) (SweepResult :type Hit%20Result%20Structure%20%28by%20ref%29))
  (bind _sm_portal (Variables|Default|GetSM_Portal))
  (bind _p_portal_idle_loop (Variables|VFX|GetP_Portal_Idle_Loop))
  ;[1 1 1] On Overlap - Toggle Post Process
  (Utilities|Casting|CastToFortPlayerPawn OtherActor)
  (CallFunction|PlayExpandorCollapseAudio true)
  ;;[1 1 1] FX if player is close
  (bind (_direction _open _push _reversing _audio) (Timeline|Timeline_0 Play)
    (:Update
      (Transformation|SetRelativeRotation _sm_portal (Math|Rotator|MakeRotator (* (Math|Float|SelectFloat _open _reversing (Utilities|Enum|Equal(Enum) _direction "Forward")) -1.8e+02) 0.0 1.8e+02) false false)
      (Rendering|Material|SetScalarParameterValueonMaterials _sm_portal "Open" (Math|Float|SelectFloat _open _reversing (Utilities|Enum|Equal(Enum) _direction "Forward")))
      (Rendering|Material|SetScalarParameterValueonMaterials _sm_portal "Push" (Math|Float|SelectFloat _push _reversing (Utilities|Enum|Equal(Enum) _direction "Forward")))
      (Effects|Components|ParticleSystem|SetFloatParameter _p_portal_idle_loop "Open" (Math|Float|SelectFloat _push _reversing (Utilities|Enum|Equal(Enum) _direction "Forward")))
      (Effects|Components|ParticleSystem|SetVectorParameter _p_portal_idle_loop "Open" (Math|Conversions|ToVector(Float) (Math|Float|SelectFloat _push _reversing (Utilities|Enum|Equal(Enum) _direction "Forward"))))
      (Audio|Parameter|SetFloatParameter (Variables|Sound|GetAudio_Looping) "Open" _audio :!class /Script/AudioExtensions.AudioParameterControllerInterface))))

(event OnComponentEndOverlap(Capsule)
  (params (OverlappedComponent :type Primitive%20Component%20Object%20Reference) (OtherActor :type Actor%20Object%20Reference) (OtherComp :type Primitive%20Component%20Object%20Reference) (OtherBodyIndex :type Integer))
  (bind _sm_portal (Variables|Default|GetSM_Portal))
  (bind _p_portal_idle_loop (Variables|VFX|GetP_Portal_Idle_Loop))
  ;[1 1 1] On Overlap - Toggle Post Process
  (if (<= (Utilities|Array|Length (Collision|GetOverlappingActors OtherActor "/Script/FortniteGame.FortPlayerPawn")) 0)
    (CallFunction|PlayExpandorCollapseAudio false)
    ;;[1 1 1] FX if player is close
    (Timeline|Timeline_0 Reverse)))

(event Custom|CheckOverlappingActors
  (bind _targetcapsule (Variables|Default|GetTargetCapsule))
  ;[1 1 1] Check for Overlapping Actors, and teleport again if anyone is overlapping
  (Utilities|IsValid _targetcapsule
    (:Is%20Valid
      (for _array_element (Collision|GetOverlappingActors _targetcapsule 0 :!class /Script/Engine.PrimitiveComponent)
        (bind _asbga_portal_athena (Utilities|Casting|CastToBGA_Portal_Athena (Variables|Portal|GetActorRef)))
        (Variables|Portal|SetOverlapActor _array_element _asbga_portal_athena)
        (CallFunction|TeleportOverlappingActors _asbga_portal_athena))
      (Development|PrintString "Overlapping Actor Check Finished"))))

(event Custom|TeleportOverlappingActors
  (bind _overlapactor (Variables|Portal|GetOverlapActor))
  ;;[1 1 1] Teleport Player and/or Vehicle
  ;[1 0.616 0] Unhooked Vehicle Teleporting For stranger things as the space was too tight for teleporting the vehicles and they were getting stuck into walls.
  (if (not (GameplayTags|HasMatchingGameplayTag (Ability|GetAbilitySystemComponent (Variables|Portal|GetOverlapActor)) "(TagName=\"Abilities.Generic.Teleporter.Immune\")"))
    (CallFunction|CheckShouldSpawn)
    (if (Networking|HasAuthority)
      (Utilities|IsValid _overlapactor
        (:Is%20Valid
          (bind _asfort_player_pawn (Utilities|Casting|CastToFortPlayerPawn _overlapactor)
            (:then
              (if (not (Variables|/Script/FortniteGame.FortPawn|GetbIsDBNO _asfort_player_pawn))
                (bind _output_get (Variables|Portal|SetPlayerPawn _asfort_player_pawn))
                (GameplayEffects|ApplyGameplayEffectToTarget :self (Ability|GetAbilitySystemComponent _output_get) :GameplayEffectClass (Variables|Portal|GetTeleportImmunityGE) :Target (Ability|GetAbilitySystemComponent _output_get) :Level 0.0) ; Apply Teleport Immunity
                (if (Vehicle|IsinVehicle _asfort_player_pawn)
                  (if (not (Vehicle|HasDriver (Vehicle|GetVehicleInterface _asfort_player_pawn)))
                    (Variables|Portal|SetVehicleInterface (Vehicle|GetVehicleInterface _asfort_player_pawn))
                    (CallFunction|VehicleTeleportEvent))
                  (else
                    (CallFunction|PlayTeleportFX (Transformation|GetActorLocation _asfort_player_pawn) (Variables|Portal|GetPlayerPawn))
                    (Networking|SwitchHasAuthority
                      (:Authority
                        (RemoteControlPawn|ForceDetachfromRemoteControlledPawn _asfort_player_pawn)
                        (Variables|/Script/FortniteGame.FortPawn|SetbIgnoreNextFallingDamage true _asfort_player_pawn) ; Have to do this because of the delay. Otherwise player can die from fall damage if they fall from high distance.
                        (Utilities|FlowControl|Sequence
                          (:then_0
                            (bind _asbga_portal_athena (Utilities|Casting|CastToBGA_Portal_Athena (Variables|Portal|GetActorRef)))
                            (Variables|Default|SetActorArrow (Variables|Default|GetArrow _asbga_portal_athena))
                            (Variables|Portal|SetTeleportLocation (+ (Transformation|GetWorldLocation (Variables|Default|GetActorArrow)) (* (Transformation|GetForwardVector (Variables|Default|GetActorArrow)) (Math|Random|RandomFloatinRange -2e+01 1.5e+02)))) ; Offset players so they don't all teleport on top of each other
                            (Variables|Portal|SetActorRotation (Transformation|GetWorldRotation (Variables|/Script/Engine.Actor|GetRootComponent _asbga_portal_athena)))
                            (CallFunction|CheckShouldGrantQuestReward))
                          (:then_1
                            (Parachute|EndSkydiving _asfort_player_pawn) ; End Skydiving if player was previously
                            (Transform|Actor|TeleportPlayerPawn _asfort_player_pawn (Variables|Portal|GetTeleportLocation) (Variables|Portal|GetActorRotation) true)
                            (Utilities|Time|SetTimerbyEvent (EventDispatchers|CreateEvent :!event Custom|CheckOverlappingActors) 3.0)))))))))
            (:CastFailed
              (bind _asfort_vehicle_interface (Utilities|Casting|CastToFortVehicleInterface (Variables|Portal|GetOverlapActor)))
              (bind _output_get_2 (Variables|Portal|SetVehicleInterface _asfort_vehicle_interface))
              (if (Vehicle|HasDriver _output_get_2)
                _
                (elif (== (Utilities|Array|Length (Vehicle|GetAllPassengers _output_get_2)) 0)
                  (CallFunction|VehicleTeleportEvent))))))))))

(event OnComponentBeginOverlap(OverlapCapsule)
  (params (OverlappedComponent :type Primitive%20Component%20Object%20Reference) (OtherActor :type Actor%20Object%20Reference) (OtherComp :type Primitive%20Component%20Object%20Reference) (OtherBodyIndex :type Integer) (bFromSweep :type Boolean) (SweepResult :type Hit%20Result%20Structure%20%28by%20ref%29))
  (bind _overlapactor (Variables|Portal|GetOverlapActor))
  ;;[1 1 1] Teleport Player and/or Vehicle
  ;[1 0.616 0] Unhooked Vehicle Teleporting For stranger things as the space was too tight for teleporting the vehicles and they were getting stuck into walls.
  (Variables|Portal|SetOverlapActor OtherActor)
  (if (not (GameplayTags|HasMatchingGameplayTag (Ability|GetAbilitySystemComponent (Variables|Portal|GetOverlapActor)) "(TagName=\"Abilities.Generic.Teleporter.Immune\")"))
    (CallFunction|CheckShouldSpawn)
    (if (Networking|HasAuthority)
      (Utilities|IsValid _overlapactor
        (:Is%20Valid
          (bind _asfort_player_pawn (Utilities|Casting|CastToFortPlayerPawn _overlapactor)
            (:then
              (if (not (Variables|/Script/FortniteGame.FortPawn|GetbIsDBNO _asfort_player_pawn))
                (bind _output_get (Variables|Portal|SetPlayerPawn _asfort_player_pawn))
                (GameplayEffects|ApplyGameplayEffectToTarget :self (Ability|GetAbilitySystemComponent _output_get) :GameplayEffectClass (Variables|Portal|GetTeleportImmunityGE) :Target (Ability|GetAbilitySystemComponent _output_get) :Level 0.0) ; Apply Teleport Immunity
                (if (Vehicle|IsinVehicle _asfort_player_pawn)
                  (if (not (Vehicle|HasDriver (Vehicle|GetVehicleInterface _asfort_player_pawn)))
                    (Variables|Portal|SetVehicleInterface (Vehicle|GetVehicleInterface _asfort_player_pawn))
                    (CallFunction|VehicleTeleportEvent))
                  (else
                    (CallFunction|PlayTeleportFX (Transformation|GetActorLocation _asfort_player_pawn) (Variables|Portal|GetPlayerPawn))
                    (Networking|SwitchHasAuthority
                      (:Authority
                        (RemoteControlPawn|ForceDetachfromRemoteControlledPawn _asfort_player_pawn)
                        (Variables|/Script/FortniteGame.FortPawn|SetbIgnoreNextFallingDamage true _asfort_player_pawn) ; Have to do this because of the delay. Otherwise player can die from fall damage if they fall from high distance.
                        (Utilities|FlowControl|Sequence
                          (:then_0
                            (bind _asbga_portal_athena (Utilities|Casting|CastToBGA_Portal_Athena (Variables|Portal|GetActorRef)))
                            (Variables|Default|SetActorArrow (Variables|Default|GetArrow _asbga_portal_athena))
                            (Variables|Portal|SetTeleportLocation (+ (Transformation|GetWorldLocation (Variables|Default|GetActorArrow)) (* (Transformation|GetForwardVector (Variables|Default|GetActorArrow)) (Math|Random|RandomFloatinRange -2e+01 1.5e+02)))) ; Offset players so they don't all teleport on top of each other
                            (Variables|Portal|SetActorRotation (Transformation|GetWorldRotation (Variables|/Script/Engine.Actor|GetRootComponent _asbga_portal_athena)))
                            (CallFunction|CheckShouldGrantQuestReward))
                          (:then_1
                            (Parachute|EndSkydiving _asfort_player_pawn) ; End Skydiving if player was previously
                            (Transform|Actor|TeleportPlayerPawn _asfort_player_pawn (Variables|Portal|GetTeleportLocation) (Variables|Portal|GetActorRotation) true)
                            (Utilities|Time|SetTimerbyEvent (EventDispatchers|CreateEvent :!event Custom|CheckOverlappingActors) 3.0)))))))))
            (:CastFailed
              (bind _asfort_vehicle_interface (Utilities|Casting|CastToFortVehicleInterface (Variables|Portal|GetOverlapActor)))
              (bind _output_get_2 (Variables|Portal|SetVehicleInterface _asfort_vehicle_interface))
              (if (Vehicle|HasDriver _output_get_2)
                _
                (elif (== (Utilities|Array|Length (Vehicle|GetAllPassengers _output_get_2)) 0)
                  (CallFunction|VehicleTeleportEvent))))))))))

(event Custom|VehicleTeleportEvent)

(event EventOverlappingOnBeginPlay
  (params (Actor :type Actor%20Object%20Reference))
  (bind _overlapactor (Variables|Portal|GetOverlapActor))
  ;;[1 1 1] Teleport Player and/or Vehicle
  ;[1 0.616 0] Unhooked Vehicle Teleporting For stranger things as the space was too tight for teleporting the vehicles and they were getting stuck into walls.
  (bind _output_get (Variables|Portal|SetOverlapActor Actor))
  (if (not (GameplayTags|HasMatchingGameplayTag (Ability|GetAbilitySystemComponent _output_get) "(TagName=\"Abilities.Generic.Teleporter.Immune\")"))
    (CallFunction|CheckShouldSpawn)
    (if (Networking|HasAuthority)
      (Utilities|IsValid _overlapactor
        (:Is%20Valid
          (bind _asfort_player_pawn (Utilities|Casting|CastToFortPlayerPawn _overlapactor)
            (:then
              (if (not (Variables|/Script/FortniteGame.FortPawn|GetbIsDBNO _asfort_player_pawn))
                (bind _output_get_2 (Variables|Portal|SetPlayerPawn _asfort_player_pawn))
                (GameplayEffects|ApplyGameplayEffectToTarget :self (Ability|GetAbilitySystemComponent _output_get_2) :GameplayEffectClass (Variables|Portal|GetTeleportImmunityGE) :Target (Ability|GetAbilitySystemComponent _output_get_2) :Level 0.0) ; Apply Teleport Immunity
                (if (Vehicle|IsinVehicle _asfort_player_pawn)
                  (if (not (Vehicle|HasDriver (Vehicle|GetVehicleInterface _asfort_player_pawn)))
                    (Variables|Portal|SetVehicleInterface (Vehicle|GetVehicleInterface _asfort_player_pawn))
                    (CallFunction|VehicleTeleportEvent))
                  (else
                    (CallFunction|PlayTeleportFX (Transformation|GetActorLocation _asfort_player_pawn) (Variables|Portal|GetPlayerPawn))
                    (Networking|SwitchHasAuthority
                      (:Authority
                        (RemoteControlPawn|ForceDetachfromRemoteControlledPawn _asfort_player_pawn)
                        (Variables|/Script/FortniteGame.FortPawn|SetbIgnoreNextFallingDamage true _asfort_player_pawn) ; Have to do this because of the delay. Otherwise player can die from fall damage if they fall from high distance.
                        (Utilities|FlowControl|Sequence
                          (:then_0
                            (bind _asbga_portal_athena (Utilities|Casting|CastToBGA_Portal_Athena (Variables|Portal|GetActorRef)))
                            (Variables|Default|SetActorArrow (Variables|Default|GetArrow _asbga_portal_athena))
                            (Variables|Portal|SetTeleportLocation (+ (Transformation|GetWorldLocation (Variables|Default|GetActorArrow)) (* (Transformation|GetForwardVector (Variables|Default|GetActorArrow)) (Math|Random|RandomFloatinRange -2e+01 1.5e+02)))) ; Offset players so they don't all teleport on top of each other
                            (Variables|Portal|SetActorRotation (Transformation|GetWorldRotation (Variables|/Script/Engine.Actor|GetRootComponent _asbga_portal_athena)))
                            (CallFunction|CheckShouldGrantQuestReward))
                          (:then_1
                            (Parachute|EndSkydiving _asfort_player_pawn) ; End Skydiving if player was previously
                            (Transform|Actor|TeleportPlayerPawn _asfort_player_pawn (Variables|Portal|GetTeleportLocation) (Variables|Portal|GetActorRotation) true)
                            (Utilities|Time|SetTimerbyEvent (EventDispatchers|CreateEvent :!event Custom|CheckOverlappingActors) 3.0)))))))))
            (:CastFailed
              (bind _asfort_vehicle_interface (Utilities|Casting|CastToFortVehicleInterface (Variables|Portal|GetOverlapActor)))
              (bind _output_get_3 (Variables|Portal|SetVehicleInterface _asfort_vehicle_interface))
              (if (Vehicle|HasDriver _output_get_3)
                _
                (elif (== (Utilities|Array|Length (Vehicle|GetAllPassengers _output_get_3)) 0)
                  (CallFunction|VehicleTeleportEvent))))))))))

(event EventNotifyTeleportedVehicle
  (params (VehicleActor :type Actor%20Object%20Reference)))

(fn CheckShouldGrantQuestReward
  (bind _vehicleinterface (Variables|Portal|GetVehicleInterface))
  (bind _returnvalue (Vehicle|GetDriver _vehicleinterface))
  (bind _playerpawn (Variables|Portal|GetPlayerPawn))
  (if (not (Variables|QuestStuff|GetCreativeMode))
    (Networking|SwitchHasAuthority
      (:Authority
        (Utilities|IsValid _playerpawn
          (:Is%20Valid
            (CallFunction|UsePortalQuestUpdate (Pawn|GetController _playerpawn)))
          (:Is%20Not%20Valid
            (Utilities|IsValid (Utilities|ToObject(Interface) _vehicleinterface)
              (:Is%20Valid
                (Utilities|IsValid _returnvalue
                  (:Is%20Valid
                    (CallFunction|UsePortalQuestUpdate (Pawn|GetController _returnvalue)))
                  (:Is%20Not%20Valid
                    _))
                (for _array_element (Vehicle|GetAllPassengers _vehicleinterface)
                  (Utilities|IsValid _array_element
                    (:Is%20Valid
                      (CallFunction|UsePortalQuestUpdate (Pawn|GetController _array_element)))))))))))))

(fn CheckShouldSpawn
  (bind _actorref (Variables|Portal|GetActorRef))
  (if (not (> (ScalableFloat|GetValueatLevel (Variables|Portal|GetShouldSpawn?)) 0.0))
    (Actor|DestroyActor)
    (Utilities|IsValid _actorref
      (:Is%20Valid
        (Actor|DestroyActor _actorref)))))

(fn OnRep_PlayerPawn)

(fn OnRep_TeleportLocation)

(fn OnRep_Vehicle)

(fn PlayExpandOrCollapseAudio
  (params (Expand :type Boolean))
  (if (Variables|Portal|GetPortalOpen)
    (Audio|PlaySoundatLocation (select Expand (Variables|Sound|GetSoundOnExpand) (Variables|Sound|GetSoundOnCollapse)) (Transformation|GetWorldLocation (Variables|Default|GetSM_Portal)))))

(fn UsePortalQuestUpdate
  (params (InController :type Controller%20Object%20Reference))
  (Utilities|IsValid InController
    (:Is%20Valid
      (bind _asfort_player_controller (Utilities|Casting|CastToFortPlayerController InController))
      (bind (_questisvalid _questcompleted) (Default|AthenaCheckQuestandObjectiveCompletion _asfort_player_controller (Variables|QuestStuff|GetQuestItem) (Variables|QuestStuff|GetObjectiveBackendName)))
      (if (and _questisvalid (not _questcompleted))
        (Class|AthenaQuestFunctionLibrary|AthenaGiveQuestUpdatetoPlayers :InQuestItemRef (Variables|QuestStuff|GetQuestItem) :ObjectiveBackendName (Variables|QuestStuff|GetObjectiveBackendName) :PlayerControllersForUpdate (Utilities|Array|MakeArray _asfort_player_controller))))))

(fn ConstructionScript)
